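# Sound cue for the Jingle emote, layering its music loop and sound effects on the first-person emote music sound class and concurrency rule.
# A mixer plays a slightly lowered one-shot SFX intro, a looping SFX layer and the looping music together.
# An enveloper fades in the SFX loop, and a continuous modulator varies its volume and pitch from the EmoteMoveSpeed parameter.
# The cue also sends to the music audio-analysis submix.
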SoundCue /BRCosmetics/Sounds/Emotes/Jingle/Emote_Jingle_MX_SFX_Loop_Cue
root=Mixer_0
nodes [
 {
  id=Enveloper_0
  children [WavePlayer_1]
  properties {pitchCurve {editorCurveData {keys [{time=0}]}}, volumeCurve {editorCurveData {keys [<at 0 0>,<at 0.5946 0.1429>,<at 0.803 1>]}}}
 }
 {id=Mixer_0, children [WavePlayer_2, ModulatorContinuous_0, WavePlayer_0], properties {inputVolume [0.7,1,1]}}
 {
  id=ModulatorContinuous_0
  children [Enveloper_0]
  properties {
   pitchModulationParams {maxOutput=1.1, minOutput=0.9, parameterName=EmoteMoveSpeed}
   volumeModulationParams {minOutput=0.5, parameterName=EmoteMoveSpeed}
  }
 }
 {id=WavePlayer_0, properties {bLooping=true, soundWaveAssetPtr=/Game/Athena/Sounds/Emotes/Jingle/Emote_Jingle_MX_Loop}}
 {id=WavePlayer_1, properties {bLooping=true, soundWaveAssetPtr=&/Emote_Jingle_SFX_Loop}}
 {id=WavePlayer_2, properties {soundWaveAssetPtr=&/Emote_Jingle_SFX_Intro}}
]
properties {
 maxDistance=2097152
 soundClassObject=/Game/Sounds/FortSoundClassesAndModes/Fort_Music_Emote_1P_PSM
 volumeMultiplier=1
 concurrencySet [/Game/Sounds/Concurrency_Rules/Athena_EmoteMusic_1P_Conc]
 soundSubmixSends [{minSendLevel=0, soundSubmix=&Fortnite_AudioAnalysis_Music}]
}
